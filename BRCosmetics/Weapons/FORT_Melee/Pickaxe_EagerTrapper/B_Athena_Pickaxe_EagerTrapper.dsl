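; Blueprint for the EagerTrapper pickaxe cosmetic, driving variant-aware swing VFX on top of its parent weapon class.
; On equip, it resolves the instigator's active Particle variant channel index from six property tags and stores it as Variant.
; All four swing events pass the montage section index to the swing VFX component, using -1 on cast failure.
; Only SwingLeft2 also spawns a Niagara trail on the weapon mesh, passing it Variant as a float.
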(event EventSwingLeft2
  (bind _returnvalue (CallFunction|GetSwingVFXComponent))
  ;[0.02306 0.15 0.03444 0.5] Final Swing main hand trail
  (|CallParentFunction)
  (bind _returnvalue_2 (Niagara|SpawnSystemAttached "/BRCosmetics/Weapons/FORT_Melee/Pickaxe_EagerTrapper/FX/NS_Pickaxe_EagerTrapper_Trail_01.NS_Pickaxe_EagerTrapper_Trail_01" (Variables|Visual|GetWeaponMesh) "None" "0, 0, 0" "0, 0, 0" "KeepRelativeOffset" true))
  (Effects|Components|ParticleSystem|SetFloatParameter _returnvalue_2 "Variant" (Math|Conversions|ToFloat(Integer) (Variables|Default|GetVariant)))
  ; SwingSectionIndex
  (bind _section_index (Utilities|SetupSwingMontageSectionIndex (Variables|Actor|GetInstigator))
    (:Out
      (Effects|Components|ParticleSystem|SetIntParameter _returnvalue "SwingSectionIndex" _section_index))
    (:Cast%20Failed%20Out
      (Effects|Components|ParticleSystem|SetIntParameter _returnvalue "SwingSectionIndex" -1))))

(event EventSwingRight
  (bind _returnvalue (CallFunction|GetSwingVFXComponent))
  ; SwingSectionIndex
  (|CallParentFunction)
  (bind _section_index (Utilities|SetupSwingMontageSectionIndex (Variables|Actor|GetInstigator))
    (:Out
      (Effects|Components|ParticleSystem|SetIntParameter _returnvalue "SwingSectionIndex" _section_index))
    (:Cast%20Failed%20Out
      (Effects|Components|ParticleSystem|SetIntParameter _returnvalue "SwingSectionIndex" -1))))

(event EventSwingLeft
  (bind _returnvalue (CallFunction|GetSwingVFXComponent))
  ; SwingSectionIndex
  (|CallParentFunction)
  (bind _section_index (Utilities|SetupSwingMontageSectionIndex (Variables|Actor|GetInstigator))
    (:Out
      (Effects|Components|ParticleSystem|SetIntParameter _returnvalue "SwingSectionIndex" _section_index))
    (:Cast%20Failed%20Out
      (Effects|Components|ParticleSystem|SetIntParameter _returnvalue "SwingSectionIndex" -1))))

(event EventSwingRight2
  (bind _returnvalue (CallFunction|GetSwingVFXComponent))
  ; SwingSectionIndex
  (|CallParentFunction)
  (bind _section_index (Utilities|SetupSwingMontageSectionIndex (Variables|Actor|GetInstigator))
    (:Out
      (Effects|Components|ParticleSystem|SetIntParameter _returnvalue "SwingSectionIndex" _section_index))
    (:Cast%20Failed%20Out
      (Effects|Components|ParticleSystem|SetIntParameter _returnvalue "SwingSectionIndex" -1))))

(event FortWeapon|EventOnEquippedAndAttached
  ; Get Variant index
  (|CallParentFunction)
  (bind _asfort_player_pawn (Utilities|Casting|CastToFortPlayerPawn (Variables|Actor|GetInstigator)))
  (bind _index (CosmeticLoadout|GetVariantChannelActiveIndexforSlotActor "(TagName=\"LoadoutSlot.Pickaxe\")" _asfort_player_pawn "(TagName=\"Cosmetics.Variant.Channel.Particle\")" (Utilities|Array|MakeArray "(TagName=\"Cosmetics.Variant.Property.Particle1\")" "(TagName=\"Cosmetics.Variant.Property.Particle2\")" "(TagName=\"Cosmetics.Variant.Property.Particle3\")" "(TagName=\"Cosmetics.Variant.Property.Particle4\")" "(TagName=\"Cosmetics.Variant.Property.Particle5\")" "(TagName=\"Cosmetics.Variant.Property.Particle6\")")))
  (Variables|Default|SetVariant _index))

(fn ConstructionScript
  (|CallParentFunction))
