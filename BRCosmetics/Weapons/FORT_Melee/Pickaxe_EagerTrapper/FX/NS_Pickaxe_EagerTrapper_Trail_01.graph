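# Niagara swing trail effect for the EagerTrapper pickaxe, with cosmetic variants chosen by User.Variant.
# MainTrails and SecondaryTrail draw skeletal-mesh anim-trail ribbons, with the secondary picking bone pairs per variant.
# Particles_Swing spawns short-lived rising sprites around a socket, with per-variant color and size, including zero-scaled variants.
# User.SwingSectionIndex selects age-based trail fades and particle start delay, while time of day tints colors.
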NiagaraSystem /BRCosmetics/Weapons/FORT_Melee/Pickaxe_EagerTrapper/FX/NS_Pickaxe_EagerTrapper_Trail_01
dependencies {
 usedDataInterfaces [
  &NiagaraDataInterfaceCurve
  &NiagaraDataInterfaceColorCurve
  &NiagaraDataInterfaceArrayFloat
  &NiagaraDataInterfaceSkeletalMesh
 ]
 usedDynamicInputs [&SetBoolByIntComparison]
 usedModules [
  &self:SystemSpawnScript.SystemScriptSource.SystemScriptGraph.NiagaraNodeAssignment_2.SetVariables_#1
  &self:SystemSpawnScript.SystemScriptSource.SystemScriptGraph.NiagaraNodeAssignment_6.SetVariables_#2
  &self:SystemSpawnScript.SystemScriptSource.SystemScriptGraph.NiagaraNodeAssignment_1.SetVariables_#3
  &self:SystemSpawnScript.SystemScriptSource.SystemScriptGraph.NiagaraNodeAssignment_3.SetVariables_#4
  &self:SystemSpawnScript.SystemScriptSource.SystemScriptGraph.NiagaraNodeAssignment_4.SetVariables_#5
  &self:SystemSpawnScript.SystemScriptSource.SystemScriptGraph.NiagaraNodeAssignment_5.SetVariables_#6
  &SystemState
  &self:SystemSpawnScript.SystemScriptSource.SystemScriptGraph.NiagaraNodeAssignment_0.SetVariables_#7
 ]
}
emitters {
 MainTrails {
  emitterUpdateScript [
   &EmitterState(Life Cycle Mode=::NewEnumerator1, Loop Behavior=::NewEnumerator1, Loop Delay=0.05, Loop Duration=0.2, UseLoopDelay=false)
   &FN_SpawnAnimTrail_v2()
  ]
  particleSpawnScript [
   /Game/Effects/Niagara/Modules/AnimTrail/FN_InitializeAnimTrail/FN_InitializeAnimTrailBoneIndecies(Second Bone or Socket Index=1)
   &FN_InitializeAnimTrail(
    AnimTrailWidth=&MergeHelper_FloatSwitch(
     0=1
     1=1
     2=1
     3=1
     4=1
     5=-1
     Index=&MakeIntFromFloat(Float=$User.Variant, Conversion Mode=::NewEnumerator1)
     Num=::NewEnumerator4
    )
    Color=&MergeHelper_ColorSwitch(
     0=<rgba 0.6927 0.6927 0.6927 1>
     1=<rgba 0.6927 0.6927 0.6927 1>
     2=<rgba 1 0.3937 0.1232 1>
     3=<rgba 0.1743 0.5356 0.6927 1>
     4=<rgba 2 2 2 1>
     Index=&MakeIntFromFloat(Float=$User.Variant, Conversion Mode=::NewEnumerator1)
     Num=::NewEnumerator3
    )
    Lifetime=0.2
    SkeletalMesh=$System.Skeletal Mesh
   ) {module=FN_InitializeAnimTrail001}
  ]
  particleUpdateScript [
   &FN_UpdateAnimTrail(SkeletalMesh=$System.Skeletal Mesh)
   &ScaleColor(
    Color Value To Scale=$StackContext.Color
    Scale Alpha=&FloatFromCurve(
     FloatCurve=DataInterface() {propertyValues {Curve {keys [{interpMode=RCIM_Cubic, tangentMode=RCTM_User, time=0, value=1}, {interpMode=RCIM_Cubic, tangentMode=RCTM_User, time=1, value=0}]}}}
     Scale Curve=&Multiply_Float(
      A=0.3
      B=&MergeHelper_FloatSwitch(
       0=&FloatFromCurve(FloatCurve=DataInterface() {propertyValues {Curve {keys [<at 0.1 1>,<at 0.18 0>]}}}, CurveIndex=$System.Age)
       1=1
       2=&FloatFromCurve(FloatCurve=DataInterface() {propertyValues {Curve {keys [<at 0.19 1>,<at 0.25 0>]}}}, CurveIndex=$System.Age)
       3=&FloatFromCurve(FloatCurve=DataInterface() {propertyValues {Curve {keys [<at 0.15 1>,<at 0.25 0>]}}}, CurveIndex=$System.Age)
       Index=$User.SwingSectionIndex
       Num=::NewEnumerator2
      )
     )
    )
    Scale RGB=&MakeVectorFromLinearColorRGB(Color=&LerpLinearColors(StartColor=$System.TODColor, EndColor=<rgba 1 1 1 1>, LerpFactor=0.5))
    ScaleA=true
    ScaleRGB=true
   )
   /SetVariables(Particles.OverlayTrailColor:LinearColor=&MultiplyLinearColor(
     Linear Color=&MergeHelper_ColorSwitch(
      0=&ColorFromCurve(ColorCurve=DataInterface() {propertyValues {
         AlphaCurve {keys [<at 0 0.35 RCIM_Cubic>,<at 0.6161 0 RCIM_Cubic>]}
         BlueCurve {keys [<at 0 0.75 RCIM_Cubic>,<at 1 0.2 RCIM_Cubic>]}
         GreenCurve {keys [<at 0 0.75 RCIM_Cubic>,<at 1 0.2 RCIM_Cubic>]}
         RedCurve {keys [<at 0 0.75 RCIM_Cubic>,<at 1 0.2 RCIM_Cubic>]}
        }})
      1=&ColorFromCurve(ColorCurve=DataInterface() {propertyValues {
         AlphaCurve {keys [<at 0 0.35 RCIM_Cubic>,<at 0.6161 0 RCIM_Cubic>]}
         BlueCurve {keys [<at 0 0.75 RCIM_Cubic>,<at 1 0.2 RCIM_Cubic>]}
         GreenCurve {keys [<at 0 0.75 RCIM_Cubic>,<at 1 0.2 RCIM_Cubic>]}
         RedCurve {keys [<at 0 0.75 RCIM_Cubic>,<at 1 0.2 RCIM_Cubic>]}
        }})
      2=&ColorFromCurve(ColorCurve=DataInterface() {propertyValues {
         AlphaCurve {keys [<at 0 0.35 RCIM_Cubic>,<at 0.6161 0 RCIM_Cubic>]}
         BlueCurve {keys [<at 0 0.9078 RCIM_Cubic>,<at 1 0.6016 RCIM_Cubic>]}
         GreenCurve {keys [<at 0 1.3797 RCIM_Cubic>,<at 1 0.965 RCIM_Cubic>]}
         RedCurve {keys [<at 0 2.1 RCIM_Cubic>,<at 1 1.5 RCIM_Cubic>]}
        }})
      3=&ColorFromCurve(ColorCurve=DataInterface() {propertyValues {
         AlphaCurve {keys [<at 0 0.35 RCIM_Cubic>,<at 0.6161 0 RCIM_Cubic>]}
         BlueCurve {keys [<at 0 2.4 RCIM_Cubic>,<at 1 1.5 RCIM_Cubic>]}
         GreenCurve {keys [<at 0 1.7357 RCIM_Cubic>,<at 1 1.1763 RCIM_Cubic>]}
         RedCurve {keys [<at 0 0.2085 RCIM_Cubic>,<at 1 0.3832 RCIM_Cubic>]}
        }})
      4=&Multiply_LinearColorByFloat(
       Linear Color=&ColorFromCurve(ColorCurve=DataInterface() {propertyValues {
          AlphaCurve {keys [<at 0 0.35 RCIM_Cubic>,<at 0.635 0 RCIM_Cubic>]}
          BlueCurve {keys [<at 0 1.5 RCIM_Cubic>,<at 1 1 RCIM_Cubic>]}
          GreenCurve {keys [<at 0 1.5 RCIM_Cubic>,<at 1 1 RCIM_Cubic>]}
          RedCurve {keys [<at 0 1.5 RCIM_Cubic>,<at 1 1 RCIM_Cubic>]}
         }})
       Float=&FNTimeOfDayScalar(Night Scale=0.65, Morning Scale=0.9, Evening Scale=0.8)
      )
      5=&ColorFromCurve(ColorCurve=DataInterface() {propertyValues {
         AlphaCurve {keys [<at 0 0.35 RCIM_Cubic>,<at 0.635 0 RCIM_Cubic>]}
         BlueCurve {keys [<at 0 1.5 RCIM_Cubic>,<at 1 1 RCIM_Cubic>]}
         GreenCurve {keys [<at 0 1.5 RCIM_Cubic>,<at 1 1 RCIM_Cubic>]}
         RedCurve {keys [<at 0 1.5 RCIM_Cubic>,<at 1 1 RCIM_Cubic>]}
        }})
      Index=&MakeIntFromFloat(Float=$User.Variant, Conversion Mode=::NewEnumerator1)
      Num=::NewEnumerator4
     )
     Scale Factor=/Niagara/DynamicInputs/Multiply/Multiply_Vector4_ByFloat(
      Vector4:Vector4f=&MakeVector4_FromLinearColor(LinearColor=&LerpLinearColors(StartColor=$System.TODColor, EndColor=<rgba 1 1 1 1>, LerpFactor=0.5))
      Float:NiagaraFloat=&MergeHelper_FloatSwitch(
       0=&FloatFromCurve(FloatCurve=DataInterface() {propertyValues {Curve {keys [<at 0.1 1>,<at 0.18 0>]}}}, CurveIndex=$System.Age)
       1=1
       2=&FloatFromCurve(FloatCurve=DataInterface() {propertyValues {Curve {keys [<at 0.19 1>,<at 0.25 0>]}}}, CurveIndex=$System.Age)
       3=&FloatFromCurve(FloatCurve=DataInterface() {propertyValues {Curve {keys [<at 0.15 1>,<at 0.25 0>]}}}, CurveIndex=$System.Age)
       Index=$User.SwingSectionIndex
       Num=::NewEnumerator2
      )
     )
    )) {isSetParameters=true}
  ]
  rendererClasses [&NiagaraRibbonRendererProperties]
  renderers [
   {rendererClass=&NiagaraRibbonRendererProperties}
   {rendererClass=&NiagaraRibbonRendererProperties, rendererIndex=1}
   {rendererClass=&NiagaraRibbonRendererProperties, rendererIndex=2}
   {rendererClass=&NiagaraRibbonRendererProperties, rendererIndex=3}
   {rendererClass=&NiagaraRibbonRendererProperties, rendererIndex=4}
  ]
 }
 Particles_Swing {
  emitterUpdateScript [
   &EmitterState(
    Life Cycle Mode=::NewEnumerator1
    Loop Behavior=::NewEnumerator1
    Loop Delay=&MergeHelper_FloatSwitch(0=0.02, 1=0.02, 2=0.02, 3=0.07, Index=$User.SwingSectionIndex, Num=::NewEnumerator2)
    Loop Duration=0.2
    UseLoopDelay=true
   )
   &SpawnRate(SpawnRate=50, Use Spawn Probability=false)
  ]
  particleSpawnScript [
   &InitializeParticle(
    Lifetime Max=0.5
    Lifetime Min=0.3
    Lifetime Mode=::NewEnumerator1
    Sprite Rotation Mode=::NewEnumerator2
    Sprite Size Mode=::NewEnumerator1
    Uniform Sprite Size Max=7
    Uniform Sprite Size Min=3
    UsePositionOffset=false
    Write Material Random=true
   )
   &SocketLocation(SamplePosOffset=false, Socket Reader=$System.Socket Reader, Socket Sampling Mode=::NewEnumerator2, VelocitySpeedLimit=false)
   &ShapeLocation(
    MaskSpawnGroup=false
    Offset Mode=::NewEnumerator0
    Override Seed=false
    Rotation Axis=<0 1 0>
    Rotation Mode=::NewEnumerator1
    Sphere Radius=20
    UseOwnerScale=true
   )
   &AddVelocity(Velocity=&RandomRangeVector(Minimum=<0 0 20>, Maximum=<0 0 35>))
  ]
  particleUpdateScript [
   &ParticleState()
   &SolveForcesAndVelocity(Clamp Velocity=false, Limit Acceleration=false)
   &ScaleColor(
    Scale Alpha=&FloatFromCurve(FloatCurve=DataInterface() {propertyValues {Curve {
        defaultValue=1
        keys [
         {arriveTangent=9.0725, interpMode=RCIM_Cubic, leaveTangent=9.0725, tangentMode=RCTM_User, time=0, value=0}
         {arriveTangent=-0.4317, interpMode=RCIM_Cubic, leaveTangent=-0.4317, tangentMode=RCTM_User, time=0.1408, value=0.9847}
         {arriveTangent=-5.9062, interpMode=RCIM_Cubic, leaveTangent=-5.9062, tangentMode=RCTM_User, time=1, value=0}
        ]
       }}})
    Scale RGB=&Multiply_VectorByFloat(
     Vector=&MakeVectorFromLinearColorRGB(Color=&MergeHelper_ColorSwitch(
       0=<rgba 0.2628 0.1732 1 1>
       1=<rgba 0.8362 0.1061 1 1>
       2=<rgba 0.4751 0.3318 0.0615 1>
       3=<rgba 1 1 1 1>
       4=<rgba 0.4751 0.3318 0.0615 1>
       5=&ColorFromCurve(
        ColorCurve=DataInterface() {propertyValues {
          AlphaCurve {keys [<at 0 0.5 RCIM_Cubic>]}
          BlueCurve {keys [<at 0.4274 3 RCIM_Cubic>,<at 0.4699 3 RCIM_Cubic>]}
          GreenCurve {keys [<at 0.4274 0.25 RCIM_Cubic>,<at 0.4699 1.0705 RCIM_Cubic>]}
          RedCurve {keys [<at 0.4274 2.0919 RCIM_Cubic>,<at 0.4699 0.3125 RCIM_Cubic>]}
         }}
        CurveIndex=&RandomRangeFloat(Minimum=0, Maximum=1)
       )
       Index=&MakeIntFromFloat(Float=$User.Variant, Conversion Mode=::NewEnumerator1)
       Num=::NewEnumerator4
      ))
     Float=&FNTimeOfDayScalar(Night Scale=0.5, Morning Scale=0.8, Evening Scale=0.65)
    )
    ScaleA=true
    ScaleRGB=true
   )
   &ScaleSpriteSize(
    Uniform Curve Scale=&SelectFloatFromArray(
     Direct Array Index=&MakeIntFromFloat(Float=$User.Variant)
     Float Selection Array=DataInterface() {propertyValues {FloatData [0,0,2,1.5,2,1.4]}}
    )
    Uniform Curve Sprite Scale=DataInterface() {propertyValues {Curve {keys [
        {arriveTangent=8.2127, interpMode=RCIM_Cubic, leaveTangent=8.2127, tangentMode=RCTM_User, time=0, value=0}
        {
         arriveTangent=0.0908
         arriveTangentWeight=0.232
         interpMode=RCIM_Cubic
         leaveTangent=0.0908
         leaveTangentWeight=2.2538
         tangentMode=RCTM_User
         time=0.1
         value=1
        }
        {arriveTangent=-0.042, interpMode=RCIM_Cubic, leaveTangent=-0.042, tangentMode=RCTM_User, time=1, value=0}
       ]}}}
   ) {module=ScaleSpriteSize001}
  ]
  rendererClasses [&NiagaraSpriteRendererProperties]
  renderers [
   {rendererClass=&NiagaraSpriteRendererProperties}
   {rendererClass=&NiagaraSpriteRendererProperties, rendererIndex=1}
   {rendererClass=&NiagaraSpriteRendererProperties, rendererIndex=2}
  ]
 }
 SecondaryTrail {
  emitterUpdateScript [
   &EmitterState(Life Cycle Mode=::NewEnumerator1, Loop Behavior=::NewEnumerator1, Loop Delay=0.05, Loop Duration=0.2, UseLoopDelay=false)
   &FN_SpawnAnimTrail_v2()
  ]
  particleSpawnScript [
   &FN_InitializeAnimTrailBoneIndecies(
    First Bone or Socket Index:NiagaraInt32=&MergeHelper_IntSwitch(
     0=2
     1=4
     2=2
     3=2
     4=2
     5=?
     Index=&MakeIntFromFloat(Float=$User.Variant, Conversion Mode=::NewEnumerator1)
     Num=::NewEnumerator3
    )
    Second Bone or Socket Index:NiagaraInt32=&MergeHelper_IntSwitch(
     0=3
     1=5
     2=3
     3=3
     4=3
     5=?
     Index=&MakeIntFromFloat(Float=$User.Variant, Conversion Mode=::NewEnumerator1)
     Num=::NewEnumerator3
    )
   )
   &FN_InitializeAnimTrail(
    AnimTrailWidth=&MergeHelper_FloatSwitch(
     0=1
     1=1
     2=1
     3=1
     4=0.75
     5=1
     Index=&MakeIntFromFloat(Float=$User.Variant, Conversion Mode=::NewEnumerator1)
     Num=::NewEnumerator4
    )
    Lifetime=0.14
    SkeletalMesh=$System.Skeletal Mesh
   ) {module=FN_InitializeAnimTrail001}
  ]
  particleUpdateScript [
   &FN_UpdateAnimTrail(SkeletalMesh=$System.Skeletal Mesh)
   &ScaleColor(
    Color Value To Scale=$StackContext.Color
    Scale Alpha=&FloatFromCurve(
     FloatCurve=DataInterface() {propertyValues {Curve {keys [{interpMode=RCIM_Cubic, tangentMode=RCTM_User, time=0, value=1}, {interpMode=RCIM_Cubic, tangentMode=RCTM_User, time=1, value=0}]}}}
     Scale Curve=&Multiply_Float(
      A=&MergeHelper_FloatSwitch(
       0=0.45
       1=0.14
       2=0.8
       3=0.3
       4=0.45
       5=0.1
       Index=&MakeIntFromFloat(Float=$User.Variant, Conversion Mode=::NewEnumerator1)
       Num=::NewEnumerator4
      )
      B=&MergeHelper_FloatSwitch(
       0=&FloatFromCurve(FloatCurve=DataInterface() {propertyValues {Curve {keys [<at 0.1 1>,<at 0.18 0>]}}}, CurveIndex=$System.Age)
       1=1
       2=&FloatFromCurve(FloatCurve=DataInterface() {propertyValues {Curve {keys [<at 0.19 1>,<at 0.25 0>]}}}, CurveIndex=$System.Age)
       3=&FloatFromCurve(FloatCurve=DataInterface() {propertyValues {Curve {keys [<at 0.15 1>,<at 0.25 0>]}}}, CurveIndex=$System.Age)
       Index=$User.SwingSectionIndex
       Num=::NewEnumerator2
      )
     )
    )
    Scale RGB=&MakeVectorFromLinearColorRGB(Color=&MultiplyLinearColor(
      Linear Color=&MergeHelper_ColorSwitch(
       0=&ColorFromCurve(ColorCurve=DataInterface() {propertyValues {
          AlphaCurve {keys [<at 0 0.3 RCIM_Cubic>]}
          BlueCurve {keys [<at 0 0.8125 RCIM_Cubic>]}
          GreenCurve {keys [<at 0 1.1121 RCIM_Cubic>]}
          RedCurve {keys [<at 0 1.3 RCIM_Cubic>]}
         }})
       1=&ColorFromCurve(ColorCurve=DataInterface() {propertyValues {
          AlphaCurve {keys [<at 0 0.5 RCIM_Cubic>]}
          BlueCurve {keys [<at 0 1.8321 RCIM_Cubic>]}
          GreenCurve {keys [<at 0 0.5 RCIM_Cubic>]}
          RedCurve {keys [<at 0 2 RCIM_Cubic>]}
         }})
       2=<rgba 0 0 0 2>
       3=&ColorFromCurve(ColorCurve=DataInterface() {propertyValues {
          AlphaCurve {keys [<at 0 0.3 RCIM_Cubic>]}
          BlueCurve {keys [<at 0 1.3 RCIM_Cubic>]}
          GreenCurve {keys [<at 0 1.3 RCIM_Cubic>]}
          RedCurve {keys [<at 0 1.3 RCIM_Cubic>]}
         }})
       4=&ColorFromCurve(ColorCurve=DataInterface() {propertyValues {
          AlphaCurve {keys [<at 0 0.3 RCIM_Cubic>]}
          BlueCurve {keys [<at 0 0.8 RCIM_Cubic>]}
          GreenCurve {keys [<at 0 0.8 RCIM_Cubic>]}
          RedCurve {keys [<at 0 0.8 RCIM_Cubic>]}
         }})
       5=&ColorFromCurve(ColorCurve=DataInterface() {propertyValues {
          AlphaCurve {keys [<at 0 0.5 RCIM_Cubic>]}
          BlueCurve {keys [<at 0 2 RCIM_Cubic>]}
          GreenCurve {keys [<at 0 1.8316 RCIM_Cubic>]}
          RedCurve {keys [<at 0 1.0521 RCIM_Cubic>]}
         }})
       Index=&MakeIntFromFloat(Float=$User.Variant, Conversion Mode=::NewEnumerator1)
       Num=::NewEnumerator4
      )
      Scale Factor=&MakeVector4_FromLinearColor(LinearColor=&LerpLinearColors(StartColor=$System.TODColor, EndColor=<rgba 1 1 1 1>, LerpFactor=0.55))
     ))
    ScaleA=true
    ScaleRGB=true
   )
  ]
  rendererClasses [&NiagaraRibbonRendererProperties]
  renderers [{rendererClass=&NiagaraRibbonRendererProperties}]
 }
}
summary {
 systemName=NS_Pickaxe_EagerTrapper_Trail_01
 emitters [
  {emitterName=MainTrails, rendererClasses [&NiagaraRibbonRendererProperties]}
  {emitterName=SecondaryTrail, rendererClasses [&NiagaraRibbonRendererProperties]}
  {emitterName=Particles_Swing, rendererClasses [&NiagaraSpriteRendererProperties]}
 ]
 userVariables [
  User.RibbonIDOffset {type=NiagaraInt32}
  User.SwingSectionIndex {type=NiagaraInt32, defaultValue=3}
  User.Variant {type=NiagaraFloat, defaultValue=0, description=""}
 ]
}
