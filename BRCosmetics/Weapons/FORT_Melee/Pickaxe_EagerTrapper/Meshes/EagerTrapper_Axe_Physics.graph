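# Physics asset for the EagerTrapper pickaxe mesh, covering the jaw and upper mouth of the axe head.
# Small capsules sit on a base bone and a dyn bone for each part, with the dyn capsules offset along the bone's X axis.
# Each dyn bone is constrained to its base with twist and Swing2 locked, forming a hinge limited on Swing1, with a wider range for the jaw than the upper mouth.
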PhysicsAsset /BRCosmetics/Weapons/FORT_Melee/Pickaxe_EagerTrapper/Meshes/EagerTrapper_Axe_Physics
bodies {
 base_jaw [{shapeType=Capsule, length=0.505, radius=0.505, shapeName=base_jaw_capsule, rotation=<rot 0 -90 0>}]
 base_mouth_up [{shapeType=Capsule, length=0.505, radius=0.505, shapeName=base_mouth_up_capsule, rotation=<rot 0 -90 0>}]
 dyn_jaw [{shapeType=Capsule, length=1, radius=1, shapeName=dyn_jaw_capsule, center=<14.7981 0 0>}]
 dyn_mouth_up [{shapeType=Capsule, length=1, radius=1, shapeName=dyn_mouth_up_capsule, center=<15.1786 0 0>}]
}
constraints [
 {
  bone1Name=dyn_jaw
  bone2Name=base_jaw
  swing1LimitDegrees=40
  swing2LimitDegrees=45
  swing2Motion=Locked
  twistLimitDegrees=45
  twistMotion=Locked
 }
 {
  bone1Name=dyn_mouth_up
  bone2Name=base_mouth_up
  swing1LimitDegrees=20
  swing2LimitDegrees=45
  swing2Motion=Locked
  twistLimitDegrees=45
  twistMotion=Locked
 }
]
